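# Additive unlit Niagara sprite material for the glow effect on the StallionSmoke glider.
# Particle color feeds an emissive adjustment function; opacity combines a procedural radial falloff, a fade by vertical normal, and particle alpha.
# The normal's vertical alignment blends side and above values for depth-fade distance and a toward-camera offset vector.
# A quality switch skips depth fade on Low and Medium.
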Material /Game/Gadgets/Assets/VinderTech_GliderChute/Glider_StallionSmoke/FX/Materials/M_Glider_StallionSmoke_Glow
graph {
 properties {bUsedWithNiagaraSprites=true, blendMode=BLEND_Additive, pixelDepthOffsetMode=PDOM_Legacy, shadingModel=MSM_Unlit, usedShadingModels=Unlit}
 nodes [
  n1 = ParticleColor()
  MaterialFunctionCall(Emissive=@n1.RGB) {materialFunction=&MF_FX_EmissiveAdjustments}
  n2 = CameraPositionWS()
  n3 = WorldPosition()
  n4 = VertexNormalWS()
  n5 = Constant3Vector() {constant=<rgba 0 0 1 0>}
  n6 = Power(Base=Abs(None=DotProduct(A=@n4, B=@n5)), Exp=ScalarParameter() {defaultValue=3, parameterName=Camera Offset Lerp Exponent})
  Multiply(
   A=Normalize(VectorInput=Subtract(A=@n2, B=@n3.XYZ))
   B=LinearInterpolate(
    A=ScalarParameter() {defaultValue=33, parameterName=Camera Offset from Side}
    B=ScalarParameter() {parameterName=Camera Offset from Above}
    Alpha=@n6
   )
  )
  n7 = Distance(A=TextureCoordinate(), B=Constant() {r=0.5})
  n8 = OneMinus(None=Saturate(None=Multiply(A=@n7) {constB=2}))
  Multiply(A=@n8, B=@n8)
  ParticleRadius()
  n9 = ScalarParameter() {defaultValue=4, parameterName=Falloff Exponent}
  n10 = VertexNormalWS()
  n11 = Constant3Vector() {constant=<rgba 0 0 1 0>}
  n12 = Saturate(None=ConstantBiasScale(None=DotProduct(A=@n10, B=@n11)))
  n13 = ScalarParameter() {defaultValue=4, parameterName=Vertical Fade Exponent}
  n14 = Multiply(A=Multiply(A=Power(Base=@n8, Exp=@n9), B=Power(Base=@n12, Exp=@n13)), B=@n1.A)
  n15 = DepthFade(
   Opacity=@n14
   FadeDistance=LinearInterpolate(
    A=ScalarParameter() {defaultValue=33, parameterName=Depth Fade Distance from Side}
    B=ScalarParameter() {defaultValue=9, parameterName=Depth Fade Distance from Above}
    Alpha=@n6
   )
  )
  QualitySwitch(Default=@n15, Low=@n14, High=@n15, Medium=@n14, Epic=@n15)
  ScalarParameter() {defaultValue=33, parameterName=Camera Offset}
 ]
}
