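# Additive unlit Niagara sprite material for a teardrop-shaped headlight glow on the StallionSmoke glider.
# Sprite_TearDrop stretches sprites using dynamic parameters, with Scale Factor feeding stretch and speed and only Direction X filling all three direction channels.
# World position offset also pushes sprites toward the camera.
# Particle color drives emissive; opacity combines particle alpha, a soft radial mask, vertical normal-based fade and depth fade.
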Material /Game/Gadgets/Assets/VinderTech_GliderChute/Glider_StallionSmoke/FX/Materials/M_Glider_StallionSmoke_Headlight_TeardropGlow
graph {
 properties {bUsedWithNiagaraSprites=true, blendMode=BLEND_Additive, pixelDepthOffsetMode=PDOM_Legacy, shadingModel=MSM_Unlit, usedShadingModels=Unlit}
 nodes [
  n1 = DynamicParameter() {paramNames [Particle Direction X, Particle Direction Y, Particle Direction Z, Scale Factor]}
  n2 = MaterialFunctionCall!(
   Max Stretch Factor=@n1.Scale Factor
   Min Tear Drop Width=ScalarParameter() {defaultValue=11, parameterName=Min Teardrop Width}
   Minimum Speed
   Max Speed
   ------ Particle Defaults Below ------
   Particle Speed=@n1.Scale Factor
   Particle Direction=Transform(None=MaterialFunctionCall(R=@n1.Particle Direction X, G=@n1.Particle Direction X, B=@n1.Particle Direction X) {materialFunction=&AppendMany}.RGB) {transformSourceType=TRANSFORMSOURCE_Instance}
   Particle Position
   Particle Size
   UV Add
  ) {materialFunction=/Engine/Functions/Engine_MaterialFunctions02/WorldPositionOffset/Sprite_TearDrop}
  n3 = Subtract(A=CameraPositionWS(), B=Add(A=@n2.World Position Offset, B=WorldPosition().XYZ))
  Add(A=@n2.World Position Offset, B=Multiply(A=Normalize(VectorInput=@n3), B=ScalarParameter() {defaultValue=33, parameterName=Camera Offset}))
  n4 = VertexNormalWS()
  n5 = Constant3Vector() {constant=<rgba 0 0 1 0>}
  n6 = Saturate(None=ConstantBiasScale(None=DotProduct(A=@n4, B=@n5)))
  n7 = ScalarParameter() {defaultValue=3, parameterName=Vertical Fade Exponent}
  n8 = ParticleColor()
  n9 = Constant() {r=0.5}
  n10 = Power(
   Base=SphereMask(A=@n2.Texture Coordinates, B=@n9, Radius=@n9) {hardnessPercent=0}
   Exp=ScalarParameter() {defaultValue=3, parameterName=Opacity Exponent}
  )
  DepthFade(
   Opacity=Multiply(A=Power(Base=@n6, Exp=@n7), B=Multiply(A=@n8.A, B=@n10))
   FadeDistance=ScalarParameter() {defaultValue=66, parameterName=Depth Fade Distance}
  )
  MaterialFunctionCall(Emissive=@n8.RGB) {materialFunction=&MF_FX_EmissiveAdjustments}
 ]
}
