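# Translucent unlit thruster effect material for the StallionSmoke glider, usable on Niagara mesh particles and ribbons.
# Particle color feeds emissive adjustments; a UV tip fade and axis-aligned Fresnel form a mask, overlay-blended with panning turbulence noise and scaled by particle alpha for opacity.
# Depth fade is an optional parameter blend, bypassed on mobile.
# A UV gradient along vertex normals builds a tip-push offset, and a parameter can swap UV axes.
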Material /Game/Gadgets/Assets/VinderTech_GliderChute/Glider_StallionSmoke/FX/Materials/M_Glider_StallionSmoke_Thruster
graph {
 properties {
  bUsedWithNiagaraMeshParticles=true
  bUsedWithNiagaraRibbons=true
  blendMode=BLEND_Translucent
  pixelDepthOffsetMode=PDOM_Legacy
  shadingModel=MSM_Unlit
  usedShadingModels=Unlit
 }
 nodes [
  n1 = ParticleColor()
  n2 = NamedRerouteUsage()
  n3 = NamedRerouteUsage()
  n4 = TextureCoordinate()
  n5 = LinearInterpolate(
   A=@n4
   B=MaterialFunctionCall(XYZ=@n4) {materialFunction=&Swizzle}.YX
   Alpha=ScalarParameter() {parameterName=SwizzleUV}
  )
  n6 = VectorParameter() {parameterName=Noise Scale and Speed, defaultValue=<rgba 1 0.666 33 9>}
  n7 = AppendVector(A=@n6.R, B=@n6.R)
  n8 = TextureSample(UVs=Panner(Coordinate=Multiply(A=@n5, B=@n7), Speed=AppendVector(A=@n6.B, B=@n6.B)) {bFractionalPart=true}) {samplerType=SAMPLERTYPE_LinearColor, texture=/Game/Effects/Fort_Effects/Effects/Environment/Desert/Dust/Organize/T_Turbulence_Seamless}
  n9 = Multiply(
   A=@n1.A
   B=MaterialFunctionCall(Base=Multiply(A=@n2, B=@n3), Blend=@n8.R) {materialFunction=&Blend_Overlay}.Result
  )
  n10 = Multiply(A=@n9, B=DepthFade(FadeDistance=ScalarParameter() {defaultValue=10, parameterName=DepthFade_Distance}))
  LinearInterpolate(
   A=@n9
   B=ShadingPathSwitch(Default=@n10, Deferred=@n10, Mobile=@n9) {desc=Migrated from FeatureLevelSwitch}
   Alpha=ScalarParameter() {parameterName=Use_DepthFade}
  )
  MaterialFunctionCall(Emissive=@n1.RGB) {materialFunction=&MF_FX_EmissiveAdjustments}
  n11 = TextureCoordinate()
  n12 = ComponentMask(None=LinearInterpolate(
    A=@n11
    B=MaterialFunctionCall(XYZ=@n11) {materialFunction=&Swizzle}.YX
    Alpha=ScalarParameter() {parameterName=SwizzleUV}
   )) {g=true}
  n13 = OneMinus(None=@n12)
  n14 = ScalarParameter() {parameterName=FlipMaskDirection}
  NamedRerouteDeclaration(None=Power(Base=LinearInterpolate(A=@n12, B=@n13, Alpha=@n14), Exp=ScalarParameter() {defaultValue=4, parameterName=Tip Fade Exponent})) {name=Tip Fade, nodeColor=<rgba 1 0 0.6797 1>}
  n15 = VectorParameter() {parameterName=FresnelVector, defaultValue=<rgba 0 1 0 1>}
  NamedRerouteDeclaration(None=MaterialFunctionCall(
    Axis=Transform(None=@n15.RGB) {transformSourceType=TRANSFORMSOURCE_Instance}
    Exponent=ScalarParameter() {defaultValue=3, parameterName=Exponent}
   ) {materialFunction=&AxisAlignedFresnel}.Fresnel) {name=Fresnel Fade, nodeColor=<rgba 1 0.7266 0 1>}
  n16 = TextureCoordinate()
  n17 = MaterialFunctionCall(XYZ=@n16) {materialFunction=&Swizzle}
  n18 = ScalarParameter() {parameterName=SwizzleUV}
  n19 = OneMinus(None=ComponentMask(None=LinearInterpolate(A=@n16, B=@n17.YX, Alpha=@n18)) {g=true})
  n20 = ScalarParameter() {defaultValue=-4, parameterName=Push Distance}
  NamedRerouteDeclaration(None=Multiply(A=Multiply(A=@n19, B=@n20), B=VertexNormalWS())) {name=Tip Push, nodeColor=<rgba 0 1 0.8125 1>}
  NamedRerouteUsage()
 ]
}
